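# Niagara system for the SpireTendFlight emote, rendering thruster flames, bursts, wake, smoke, speed lines and a light.
# User velocity and FireIntensity thresholds trigger bursts and vary spawn rates, size and color, while UseIdleVFX gates visibility.
# Thruster bursts, wake, smoke and light sample skeletal mesh bones; wake and smoke blend between local and world space.
# Time-of-day helpers tune brightness, and camera-angle dot products reshape speed lines.
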NiagaraSystem /BRCosmetics/Animation/Game/MainPlayer/Emotes/SpireTendFlight/VFX/NS_Emote_SpireTendFlight
dependencies {
 usedDataInterfaces [
  &NiagaraDataInterfaceCurve
  &NiagaraDataInterfaceColorCurve
  &NiagaraDataInterfaceVectorCurve
  &NiagaraDataInterfaceCamera
  &NiagaraDataInterfaceVector2DCurve
  &NiagaraDataInterfaceSpriteRendererInfo
  &NiagaraDataInterfaceSkeletalMesh
  &NiagaraDataInterfaceArrayFloat
 ]
 usedDynamicInputs [
  &SimulationPosition
  &MakeFloatFromVector2d
  &Multiply_Vector2
  &Add_Float
  &Abs_Float
 ]
 usedModules [
  &self:SystemSpawnScript.SystemScriptSource.SystemScriptGraph.NiagaraNodeAssignment_0.SetVariables_#1
  &self:SystemSpawnScript.SystemScriptSource.SystemScriptGraph.NiagaraNodeAssignment_1.SetVariables_#2
  &SystemState
  &BlendSpaceHelper
  &Timeline
  /BRCosmetics/Characters/Player/Female/Medium/Bodies/F_MED_PiperShelf_Pearl/FX/TimeOfDayGlobal
  &self:SystemSpawnScript.SystemScriptSource.SystemScriptGraph.NiagaraNodeAssignment_2.SetVariables_#3
  &self:SystemSpawnScript.SystemScriptSource.SystemScriptGraph.NiagaraNodeAssignment_3.SetVariables_#4
  &InterpolateOverTime
  &self:SystemSpawnScript.SystemScriptSource.SystemScriptGraph.NiagaraNodeAssignment_4.SetVariables_#5
 ]
}
emitters {
 BurstElements {
  emitterUpdateScript [
   &EmitterState(
    Inactive Response=::NewEnumerator1
    Life Cycle Mode=::NewEnumerator1
    Loop Behavior=::NewEnumerator1
    Loop Delay=0.2
    Loop Duration Mode=::NewEnumerator1
    UseLoopDelay=false
   )
   /Niagara/Modules/Emitter/SpawnByTrigger(
    Trigger:NiagaraBool=&SetBoolByFloatComparison(A=&VectorLength(Vector=$User.OwnerLocalVelocity), B=20)
    TriggerIntervalSwitch=false
   )
   &SpawnByTrigger(Trigger:NiagaraBool=&SetBoolByFloatComparison(A=$User.FireIntensity, B=0.01), TriggerIntervalSwitch=false) {module=SpawnByTrigger001}
   &SpawnByTrigger(
    Spawn Group=1
    Trigger:NiagaraBool=&SetBoolByFloatComparison(A=&VectorLength(Vector=$User.OwnerLocalVelocity), B=20)
    TriggerIntervalSwitch=false
   ) {module=SpawnByTrigger002}
   &SpawnByTrigger(
    Spawn Group=1
    Trigger:NiagaraBool=&SetBoolByFloatComparison(A=$User.FireIntensity, B=0.01)
    TriggerIntervalSwitch=false
   ) {module=SpawnByTrigger003}
  ]
  particleSpawnScript [
   /SetVariables(Particles.VisibilityTag=&SpawnGroup()) {isSetParameters=true}
   &InitializeParticle(
    Color=&MaskLinearColor(Mask Source=0, Mask Parameter=?, Mask=0, Masked Linear Color=<rgba 8.7581 4.001 20 1>, Passthrough Linear Color=<rgba 2.2733 0.8177 5 1>)
    Lifetime=0.25
    Mesh Scale=&MaskVector(
     Mask Source=0
     Mask Parameter=?
     Mask=0
     Masked Vector=&Multiply_VectorByFloat(Vector=&RandomRangeVector(Minimum=<1 2 2>, Maximum=<2 2 2>), Float=1)
     Passthrough Vector=&Multiply_VectorByFloat(Vector=&RandomRangeVector(Minimum=<0.3 1 1>, Maximum=<0.5 1.3 1.3>), Float=0.5)
    )
    Mesh Scale Mode=::NewEnumerator4
    UsePositionOffset=false
    Write Material Random=true
   ) {module=InitializeParticle001}
   &InitialMeshOrientation(Mesh Axis To Orient=<0 2 0>, Mesh Orientation Mode=::NewEnumerator2, Use Rotation Vector=false) {module=InitialMeshOrientation002}
  ]
  particleUpdateScript [
   &ParticleState()
   &SolveForcesAndVelocity(Clamp Velocity=false, Limit Acceleration=false)
   &SkeletalMeshLocation(
    Bone / Socket Index=2
    Bone Sampling Mode=::NewEnumerator4
    MaskSpawnGroup=false
    Sampled Position Offset=<0 17 0>
    Skeletal Mesh=$System.Skeletal Mesh Reader
   )
   &ScaleColor(
    Scale Alpha=&MaskFloat(
     Mask Source=0
     Mask Parameter=?
     Mask=0
     Masked Float=&Multiply_Float(
      A=&FN_FrontEndFloatModifier(Gameplay=1, Front End=3)
      B=&Multiply_Float(
       A=&FloatFromCurve(FloatCurve=DataInterface() {propertyValues {Curve {keys [<at 0 0 RCIM_Cubic>,<at 0.026 1 RCIM_Cubic>,<at 1 0>]}}})
       B=&FNTimeOfDayScalar(Night Scale=0.3, Morning Scale=0.2, Evening Scale=0.2)
      )
     )
     Passthrough Float=&Multiply_Float(
      A=&Multiply_Float(
       A=&FNTimeOfDayScalar(Night Scale=0.4, Morning Scale=0.2, Evening Scale=0.2)
       B=&FloatFromCurve(FloatCurve=DataInterface() {propertyValues {Curve {keys [<at 0 0>,<at 0.4 0.1>,<at 1 1>]}}}, CurveIndex=$System.Timeline.NormalizedTime)
      )
      B=&FloatFromCurve(FloatCurve=DataInterface() {propertyValues {Curve {keys [
           {interpMode=RCIM_Cubic, tangentMode=RCTM_User, time=0, value=0}
           {interpMode=RCIM_Cubic, tangentMode=RCTM_User, time=0.5, value=1}
           {tangentMode=RCTM_User, time=1, value=0}
          ]}}})
     )
    )
    Scale RGB=&MaskVector(Mask Source=0, Mask Parameter=?, Mask=0, Masked Vector=<3 3 3>, Passthrough Vector=<1 1 1>)
    ScaleA=true
    ScaleRGB=true
   ) {module=ScaleColor001}
   &ScaleMeshSize(Scale Factor=&Multiply_VectorByFloat(
     Vector=&MaskVector(
      Mask Source=0
      Mask Parameter=?
      Mask=0
      Masked Vector=&VectorFromCurve(
       VectorCurve=DataInterface() {propertyValues {XCurve {keys [<at 0 0.5>,<at 1 1.5>]}, YCurve {keys [<at 0 0.7>,<at 1 1.6>]}, ZCurve {keys [<at 0 0.7>,<at 1 2>]}}}
       Scale Curve=<0.8 0.8 0.8>
      )
      Passthrough Vector=&Multiply_VectorByFloat(
       Vector=<0.75 0.85 0.85>
       Float=&FloatFromCurve(FloatCurve=DataInterface() {propertyValues {Curve {keys [{interpMode=RCIM_Cubic, tangentMode=RCTM_User, time=0, value=0.4}, {interpMode=RCIM_Cubic, tangentMode=RCTM_User, time=1, value=1}]}}})
      )
     )
     Float=&MakeCustomFloatFromBool(
      Bool=&SetBoolByFloatComparison(A=$User.UseIdleVFX, B=1, Comparison Type=::NewEnumerator6)
      True Float=0
      False Float=1
     )
    ))
   &DynamicMaterialParameters(
    Index 0 Param 1=&FloatFromCurve(FloatCurve=DataInterface() {propertyValues {Curve {keys [<at 0 0 RCIM_Cubic>,<at 1 0.28 RCIM_Cubic>]}}})
    Index 0 Param 2=&RandomRangeFloat(Minimum=0, Maximum=1)
    Index 0 Param 3=100
    Index 0 Param 4=&FloatFromCurve(FloatCurve=DataInterface() {propertyValues {Curve {keys [<at 0 0>, {arriveTangent=-3.0374, leaveTangent=2.0371, time=0.2, value=0}, {interpMode=RCIM_Cubic, tangentMode=RCTM_Break, time=1, value=0}]}}})
    Module.Param0WriteEnabled=true
    Module.Param1WriteEnabled=true
    Module.Param2WriteEnabled=true
    Module.Param3WriteEnabled=true
   ) {enabled=false}
   /SetVariables(Particles.Vector 4:Vector4f=&MakeVector4(
     X=&FloatFromCurve(FloatCurve=DataInterface() {propertyValues {Curve {keys [<at 0 0 RCIM_Cubic>,<at 1 0.28 RCIM_Cubic>]}}})
     Y=&RandomRangeFloat(Minimum=0, Maximum=1)
     Z=100
     W=&FloatFromCurve(FloatCurve=DataInterface() {propertyValues {Curve {keys [<at 0 0>, {arriveTangent=-3.0374, leaveTangent=2.0371, time=0.2, value=0}, {interpMode=RCIM_Cubic, tangentMode=RCTM_Break, time=1, value=0}]}}})
    )) {isSetParameters=true}
  ]
  rendererClasses [&NiagaraMeshRendererProperties]
  renderers [
   {rendererClass=&NiagaraMeshRendererProperties, rendererIndex=0}
   {rendererClass=&NiagaraMeshRendererProperties, rendererIndex=1}
  ]
 }
 LargeWake {
  emitterUpdateScript [
   &EmitterState(
    Inactive Response=::NewEnumerator1
    Life Cycle Mode=::NewEnumerator1
    Loop Behavior=::NewEnumerator1
    Loop Delay=0.05
    Loop Duration Mode=::NewEnumerator1
    UseLoopDelay=true
   )
   &SpawnRate(
    SpawnRate=&MakeCustomFloatFromBool(
     Bool=&Boolean_NotOperation(Bool=&SetBoolByFloatComparison(A=$User.UseIdleVFX, B=1, Comparison Type=::NewEnumerator6))
     True Float=&MakeCustomFloatFromBool(
      Bool=&Boolean_OrOperation(
       Bool1=&SetBoolByFloatComparison(A=&VectorLength(Vector=$User.OwnerLocalVelocity), B=100)
       Bool2=&SetBoolByFloatComparison(A=$User.FireIntensity, B=0.1)
      )
      True Float=7
      False Float=4
     )
     False Float=0
    )
    Use Spawn Probability=false
   )
  ]
  particleSpawnScript [
   &InitializeParticle(
    Color=&FNTimeOfDayColor(
     Night Color=<rgba 0.3445 0.3815 0.4823 1>
     Morning Color=<rgba 0.4823 0.4823 0.4823 1>
     Day Color=<rgba 0.4823 0.4823 0.4823 1>
     Evening Color=<rgba 0.4823 0.4823 0.4823 1>
    )
    Lifetime Max=1.2
    Lifetime Min=0.7
    Lifetime Mode=::NewEnumerator1
    Mesh Scale=&Multiply_VectorByFloat(
     Vector=&Multiply_VectorByFloat(
      Vector=&RandomRangeVector(Minimum=<1 1 0.5>, Maximum=<1 1 1>)
      Float=&RandomRangeFloat(Minimum=0.65, Maximum=1.1)
     )
     Float=1.25
    )
    Mesh Scale Mode=::NewEnumerator4
    UsePositionOffset=true
    Write Material Random=true
   )
   &InitialMeshOrientation(
    Mesh Orientation Mode=::NewEnumerator2
    Orientation Coordinate Space=::Local
    Rotation=<0 0 0.25>
    Rotation Coordinate Space=::NewEnumerator1
    Use Rotation Vector=true
   ) {module=InitialMeshOrientation001}
   &SkeletalMeshLocation(
    Bone / Socket Index=3
    Bone Sampling Mode=::NewEnumerator4
    MaskSpawnGroup=false
    Orientation Sampling=::NewEnumerator1
    Position Sampling=::NewEnumerator1
    Sampled Position Offset=<0 -50 -53>
    Skeletal Mesh=$System.Skeletal Mesh Reader
   ) {module=SkeletalMeshLocation001}
   &AddVelocity(Velocity=&MakeCustomVectorFromBool(
     Bool=&Boolean_OrOperation(
      Bool1=&SetBoolByFloatComparison(A=&VectorLength(Vector=$User.OwnerLocalVelocity), B=100)
      Bool2=&SetBoolByFloatComparison(A=$User.FireIntensity, B=0.1)
     )
     True Vector=&RandomRangeVector(Minimum=<0 250 0>, Maximum=<0 400 0>)
     False Vector=&RandomRangeVector(Minimum=<0 125 0>, Maximum=<0 200 0>)
    ))
   &ApplyInitialForces()
   &CameraOffset(Camera Offset Amount=-5)
  ]
  particleUpdateScript [
   &ParticleState()
   &DynamicMaterialParameters(
    Index 0 Param 1=&FloatFromCurve(FloatCurve=DataInterface() {propertyValues {Curve {keys [
         <at 0 0.2 RCIM_Cubic>
         {interpMode=RCIM_Cubic, tangentMode=RCTM_User, time=0.3, value=0.45}
         {interpMode=RCIM_Cubic, tangentMode=RCTM_User, time=1, value=0}
        ]}}})
    Module.Param0WriteEnabled=true
    Module.Param1WriteEnabled=true
    Module.Param2WriteEnabled=true
    Module.Param3WriteEnabled=true
   )
   &ScaleMeshSize(Scale Factor=&Multiply_VectorByFloat(
     Vector=&Multiply_VectorByFloat(
      Vector=&MakeVector(
       X=1
       Y=1
       Z=&FloatFromCurve(FloatCurve=DataInterface() {propertyValues {Curve {keys [<at 0.15 1>,<at 0.6 0>]}}})
      )
      Float=&FloatFromCurve(
       FloatCurve=DataInterface() {propertyValues {Curve {keys [
           {arriveTangent=0.9192, interpMode=RCIM_Cubic, leaveTangent=0.9192, tangentMode=RCTM_User, time=0, value=0}
           {arriveTangent=1.6059, interpMode=RCIM_Cubic, leaveTangent=1.6059, tangentMode=RCTM_User, time=0.4047, value=0.34}
           {arriveTangent=-0.0405, interpMode=RCIM_Cubic, leaveTangent=-0.0405, tangentMode=RCTM_User, time=1, value=1}
          ]}}}
       Scale Curve=2.5
      )
     )
     Float=&MakeCustomFloatFromBool(
      Bool=&SetBoolByFloatComparison(A=$User.UseIdleVFX, B=1, Comparison Type=::NewEnumerator6)
      True Float=0
      False Float=1
     )
    ))
   &ScaleColor(
    Scale Alpha=&FloatFromCurve(
     FloatCurve=DataInterface() {propertyValues {Curve {keys [
         {interpMode=RCIM_Cubic, tangentMode=RCTM_User, time=0, value=0.2}
         {interpMode=RCIM_Cubic, leaveTangent=-2.4875, tangentMode=RCTM_Break, time=0.17, value=1}
         {tangentMode=RCTM_User, time=1, value=0}
        ]}}}
     Scale Curve=&FNTimeOfDayScalar(Night Scale=0.4, Morning Scale=0.5, Day Scale=0.5, Evening Scale=0.5)
    )
    Scale RGB=&VectorFromFloat(Value=&FNTimeOfDayScalar(Night Scale=0.57, Morning Scale=0.75, Day Scale=0.75, Evening Scale=0.75))
    ScaleA=true
    ScaleRGB=true
   )
   &BlendSpace("Position Blend (World <-> Local)":NiagaraFloat=&FloatFromCurve(
     FloatCurve=DataInterface() {propertyValues {Curve {
        defaultValue=1
        keys [{interpMode=RCIM_Cubic, tangentMode=RCTM_User, time=0, value=1}, {interpMode=RCIM_Cubic, tangentMode=RCTM_User, time=1, value=0}]
       }}}
     Translation Contribution:NiagaraFloat=1
     Rotation Contribution:NiagaraFloat=1
    ))
   &SolveForcesAndVelocity(Clamp Velocity=false, Limit Acceleration=false)
  ]
  rendererClasses [&NiagaraMeshRendererProperties]
  renderers [{rendererClass=&NiagaraMeshRendererProperties, rendererIndex=0}]
 }
 Light {
  emitterUpdateScript [
   &EmitterState()
   &SpawnBurst_Instantaneous(Use Spawn Probability=false, UseLoopCountLimit=false)
  ]
  particleSpawnScript [&InitializeParticle(
    Lifetime=$Emitter.CurrentLoopDuration
    Mesh Scale=<1.8 1.8 1.8>
    Mesh Scale Mode=::NewEnumerator4
    Position Offset=<0 -85 75>
    UsePositionOffset=true
    Write Material Random=true
   )]
  particleUpdateScript [
   &ParticleState()
   &SolveForcesAndVelocity(Clamp Velocity=false, Limit Acceleration=false)
   &SkeletalMeshLocation(Bone / Socket Index=2, Bone Sampling Mode=::NewEnumerator4, MaskSpawnGroup=false, Skeletal Mesh=$System.Skeletal Mesh Reader)
   &Light_Attributes(Diffuse Scale=1, Falloff Exponent=1, Light Specular Scale=0, Radius=1, Specular Scale=1)
   /SetVariables(
    Particles.Light Color:LinearColor=&MakeCustomLinearColorFromBool(
     Bool=&Boolean_OrOperation(
      Bool1=&SetBoolByFloatComparison(A=&VectorLength(Vector=$User.OwnerLocalVelocity), B=100)
      Bool2=&SetBoolByFloatComparison(A=$User.FireIntensity, B=0.1)
     )
     True Color=<rgba 5.2837 1.8324 6 1>
     False Color=<rgba 2.7882 1.3166 4 1>
    )
    Particles.Position=&AddVectorToPosition(Position=$Particles.Position, Vector=<0 40 0>)
   ) {isSetParameters=true}
  ]
  rendererClasses [&NiagaraLightRendererProperties]
  renderers [{rendererClass=&NiagaraLightRendererProperties, rendererIndex=0}]
 }
 Smoke {
  emitterUpdateScript [
   &EmitterState(Life Cycle Mode=::NewEnumerator1, Loop Behavior=::NewEnumerator1, Loop Delay=0.05, Loop Duration Mode=::NewEnumerator1, UseLoopDelay=true)
   &SpawnRate(
    SpawnRate=&MakeCustomFloatFromBool(
     Bool=&Boolean_NotOperation(Bool=&SetBoolByFloatComparison(A=$User.UseIdleVFX, B=1, Comparison Type=::NewEnumerator6))
     True Float=&MakeCustomFloatFromBool(
      Bool=&Boolean_OrOperation(
       Bool1=&SetBoolByFloatComparison(A=&VectorLength(Vector=$User.OwnerLocalVelocity), B=100)
       Bool2=&SetBoolByFloatComparison(A=$User.FireIntensity, B=0.1)
      )
      True Float=8
      False Float=4
     )
     False Float=0
    )
    Use Spawn Probability=false
   )
  ]
  particleSpawnScript [
   /SetVariables(Particles.Random=&RandomRangeFloat(Minimum=0, Maximum=1)) {isSetParameters=true}
   &CameraOffset(Camera Offset Amount=30)
   &DynamicMaterialParameters(
    Index 0 Param 2=&UniformAOrBFloat(A=0)
    Module.Param0WriteEnabled=true
    Module.Param1WriteEnabled=true
    Module.Param2WriteEnabled=true
    Module.Param3WriteEnabled=true
   ) {module=DynamicMaterialParameters001}
   &InitializeParticle(
    Color=&FNTimeOfDayColor(
     Night Color=<rgba 0.4577 0.5174 0.599 1>
     Morning Color=<rgba 0.599 0.599 0.599 1>
     Day Color=<rgba 0.599 0.599 0.599 1>
     Evening Color=<rgba 0.599 0.599 0.599 1>
    )
    Lifetime=1.1
    Sprite Rotation Angle=-90
    Sprite Rotation Mode=::NewEnumerator0
    Sprite Size Mode=::NewEnumerator3
    Uniform Sprite Size=&Multiply_Float(A=&Lerp_Float(A=130, B=180, Alpha=$Particles.Random), B=1.25)
    UsePositionOffset=true
    Write Material Random=true
   ) {module=InitializeParticle001}
   &AddVelocity(Velocity=&MakeCustomVectorFromBool(
     Bool=&Boolean_OrOperation(
      Bool1=&SetBoolByFloatComparison(A=&VectorLength(Vector=$User.OwnerLocalVelocity), B=100)
      Bool2=&SetBoolByFloatComparison(A=$User.FireIntensity, B=0.1)
     )
     True Vector=&RandomRangeVector(Minimum=<0 250 0>, Maximum=<0 400 0>)
     False Vector=&RandomRangeVector(Minimum=<0 125 0>, Maximum=<0 200 0>)
    ))
   &ApplyInitialForces()
   &InitialMeshOrientation(
    Mesh Orientation Mode=::NewEnumerator2
    Orientation Coordinate Space=::World
    Orientation Vector=<0 0 1>
    Rotation=<0 -0.25 0>
    Use Rotation Vector=false
   )
   &SkeletalMeshLocation(
    Bone / Socket Index=3
    Bone Sampling Mode=::NewEnumerator4
    MaskSpawnGroup=false
    Orientation Sampling=::NewEnumerator1
    Position Sampling=::NewEnumerator1
    Sampled Position Offset=<0 -50 -53>
    Skeletal Mesh=$System.Skeletal Mesh Reader
   )
   &CameraOffset(Camera Offset Amount=-5) {module=CameraOffset001}
  ]
  particleUpdateScript [
   &ParticleState()
   &ScaleSpriteSize(
    Scale Sprite Size Mode=::NewEnumerator0
    Uniform Scale Factor=&Multiply_Float(
     A=&FloatFromCurve(FloatCurve=DataInterface() {propertyValues {Curve {keys [
          {arriveTangent=1.6815, interpMode=RCIM_Cubic, leaveTangent=1.6815, tangentMode=RCTM_User, time=0.2, value=0.5}
          {arriveTangent=0.1063, interpMode=RCIM_Cubic, leaveTangent=0.1063, tangentMode=RCTM_User, time=0.55, value=1}
         ]}}})
     B=&MakeCustomFloatFromBool(
      Bool=&SetBoolByFloatComparison(A=$User.UseIdleVFX, B=1, Comparison Type=::NewEnumerator6)
      True Float=0
      False Float=1
     )
    )
   )
   &SolveForcesAndVelocity(Clamp Velocity=false, Limit Acceleration=false)
   &AlignSpriteToMeshOrientation(SetAlignment=true, SetFacing=true)
   &DynamicMaterialParameters(
    Index 0 Param 1=&FloatFromCurve(FloatCurve=DataInterface() {propertyValues {Curve {keys [
         {arriveTangent=1.7661, interpMode=RCIM_Cubic, leaveTangent=1.7661, tangentMode=RCTM_User, time=0, value=0}
         {arriveTangent=0.0224, interpMode=RCIM_Cubic, leaveTangent=0.0224, tangentMode=RCTM_User, time=1, value=1}
        ]}}})
    Index 0 Param 2=1
    Module.Param0WriteEnabled=true
    Module.Param1WriteEnabled=true
    Module.Param2WriteEnabled=true
    Module.Param3WriteEnabled=true
   )
   &ScaleColor(
    Scale Alpha=&FloatFromCurve(
     FloatCurve=DataInterface() {propertyValues {Curve {keys [<at 0 0>, {tangentMode=RCTM_User, time=0.75, value=1}, {tangentMode=RCTM_User, time=1, value=0}]}}}
     Scale Curve=&FNTimeOfDayScalar(Night Scale=0.4, Morning Scale=0.5, Day Scale=0.5, Evening Scale=0.5)
    )
    Scale RGB=&Multiply_VectorByFloat(
     Vector=&MakeVectorFromLinearColorRGB(Color=&ColorFromCurve(ColorCurve=DataInterface() {propertyValues {
         AlphaCurve {keys [<at 0 0>,<at 0.4487 1 RCIM_Cubic>]}
         BlueCurve {keys [
           <at 0.1811 0.6425 RCIM_Cubic>
           {arriveTangent=-0.3057, interpMode=RCIM_Cubic, leaveTangent=-0.3057, time=0.4806, value=1}
           <at 0.7837 0.4583>
          ]}
         GreenCurve {keys [
           <at 0.1811 0.6562 RCIM_Cubic>
           {arriveTangent=-0.3285, interpMode=RCIM_Cubic, leaveTangent=-0.3285, time=0.4806, value=1}
           <at 0.7837 0.4583>
          ]}
         RedCurve {keys [
           <at 0.1811 0.6388 RCIM_Cubic>
           {arriveTangent=-0.2996, interpMode=RCIM_Cubic, leaveTangent=-0.2996, time=0.4806, value=1}
           <at 0.7837 0.4583>
          ]}
        }}))
     Float=&FNTimeOfDayScalar(Night Scale=0.57, Morning Scale=1, Evening Scale=1)
    )
    ScaleA=true
    ScaleRGB=true
   )
   &BlendSpace("Position Blend (World <-> Local)":NiagaraFloat=&FloatFromCurve(
     FloatCurve=DataInterface() {propertyValues {Curve {
        defaultValue=1
        keys [{interpMode=RCIM_Cubic, tangentMode=RCTM_User, time=0, value=1}, {interpMode=RCIM_Cubic, tangentMode=RCTM_User, time=1, value=0}]
       }}}
     Translation Contribution:NiagaraFloat=1
     Rotation Contribution:NiagaraFloat=1
    ))
  ]
  rendererClasses [&NiagaraSpriteRendererProperties]
  renderers [
   {rendererClass=&NiagaraSpriteRendererProperties}
   {rendererClass=&NiagaraSpriteRendererProperties, rendererIndex=1}
   {rendererClass=&NiagaraSpriteRendererProperties, rendererIndex=2}
  ]
 }
 SpeedLines {
  emitterUpdateScript [
   &EmitterState(
    Inactive Response=::NewEnumerator1
    Life Cycle Mode=::NewEnumerator1
    Loop Behavior=::NewEnumerator1
    Loop Duration Mode=::NewEnumerator1
    UseLoopDelay=false
   )
   &SpawnRate(
    SpawnRate=&MakeCustomFloatFromBool(
     Bool=&SetBoolByFloatComparison(A=&VectorLength(Vector=$User.OwnerLocalVelocity), B=200)
     True Float=&Lerp_Float(A=24, B=48, Alpha=$User.FireIntensity)
     False Float=&Lerp_Float(A=5, B=36, Alpha=$User.FireIntensity)
    )
    Use Spawn Probability=false
   ) {module=SpawnRate001}
   &SpawnByTrigger(
    Spawn Count=15
    Trigger:NiagaraBool=&SetBoolByFloatComparison(A=&VectorLength(Vector=$User.OwnerLocalVelocity), B=100, Comparison Type=::NewEnumerator2)
    TriggerIntervalSwitch=false
   )
  ]
  particleSpawnScript [
   &InitializeParticle(
    Color=&FNTimeOfDayColor(Night Color=<rgba 0.5875 0.6613 0.9448 1>, Morning Color=<rgba 1 1 1 1>, Day Color=<rgba 1 1 1 1>, Evening Color=<rgba 1 1 1 1>)
    Lifetime Max=0.23
    Lifetime Min=0.15
    Lifetime Mode=::NewEnumerator1
    Sprite Size=&Multiply_Vector2D_ByFloat(
     Float=&RandomRangeFloat(Minimum=1, Maximum=3)
     Vector2D=&RandomRangeVector2D(Minimum=<1 1>, Maximum=<5 4>)
    )
    Sprite Size Mode=::NewEnumerator4
    UsePositionOffset=false
    Write Material Random=true
   )
   &ShapeLocation(
    Handle Radius=0
    Large Radius=&Multiply_Float(
     A=&RandomRangeFloat(Minimum=80, Maximum=100)
     B=&RemapRange(
      Input Value To Remap=&MakeFloatFromVector(
       Vector=&VectorFromCurve(
        VectorCurve=DataInterface() {propertyValues {XCurve {keys [<at 0 1>,<at 1 0>]}, YCurve {keys [<at 0 1>,<at 1 0>]}, ZCurve {keys [<at 0 1>,<at 1 0>]}}}
        CurveIndex=&DotProduct(
         Vector A=&ReturnCameraVector(
          Camera Vectors=::NewEnumerator4
          Coordinate Space=::Simulation
          Origin Position=$Particles.Position
          CameraQuery=DataInterface() {propertyValues {bRequireCurrentFrameData=false}}
         )
         Vector B=<0 1 0>
        )
       )
       Channel=::NewEnumerator0
      )
      Input Min=0.2
      Input Max=0.9
      Output Min=1
      Output Max=0.5
     )
    )
    MaskSpawnGroup=false
    Offset=<0 -80 160>
    Offset Mode=::NewEnumerator0
    Override Seed=false
    Rotation Axis=<0 1 0>
    Rotation Mode=::NewEnumerator1
    Shape Primitive=::NewEnumerator3
    UseOwnerScale=true
   )
   &AddVelocity(Velocity=&RandomRangeVector(Minimum=<0 1000 0>, Maximum=<0 500 0>))
   &AddVelocity(
    Module.Use Falloff Distance=false
    Origin Offset=<0 -50 100>
    Override Seed=false
    Velocity Mode=::NewEnumerator1
    Velocity Speed=&Multiply_Float(
     A=400
     B=&RemapRange(
      Input Value To Remap=&MakeFloatFromVector(
       Vector=&VectorFromCurve(
        VectorCurve=DataInterface() {propertyValues {XCurve {keys [<at 0 1>,<at 1 0>]}, YCurve {keys [<at 0 1>,<at 1 0>]}, ZCurve {keys [<at 0 1>,<at 1 0>]}}}
        CurveIndex=&DotProduct(
         Vector A=&ReturnCameraVector(
          Camera Vectors=::NewEnumerator4
          Coordinate Space=::Simulation
          Origin Position=$Particles.Position
          CameraQuery=DataInterface() {propertyValues {bRequireCurrentFrameData=false}}
         )
         Vector B=<0 1 0>
        )
       )
       Channel=::NewEnumerator0
      )
      Input Min=0.2
      Input Max=0.9
      Output Min=1
      Output Max=0
     )
    )
   ) {module=AddVelocity001}
   &ApplyInitialForces()
  ]
  particleUpdateScript [
   &ParticleState()
   &ScaleSpriteSize(Scale Factor=&Multiply_Vector2D(
     A=&MakeCustomVector2DFromBool(
      Bool=&Boolean_OrOperation(
       Bool1=&SetBoolByFloatComparison(A=&VectorLength(Vector=$User.OwnerLocalVelocity), B=100)
       Bool2=&SetBoolByFloatComparison(A=$User.FireIntensity, B=0.2)
      )
      True Vector2D=<1 1>
      False Vector2D=<1 0.4>
     )
     B=&Multiply_Vector2D_ByFloat(
      Float=$User.UseIdleVFX
      Vector2D=&Vector2DFromCurve(
       Vector2Curve=DataInterface() {propertyValues {XCurve {keys [<at 0 0.5 RCIM_Cubic>,<at 0.5 1 RCIM_Cubic>]}, YCurve {keys [<at 0 0 RCIM_Cubic>,<at 0.5 1 RCIM_Cubic>]}}}
       ScaleCurve=&Vector2DFromFloat(Value=&RemapRange(
         Input Value To Remap=&MakeFloatFromVector(
          Vector=&VectorFromCurve(
           VectorCurve=DataInterface() {propertyValues {XCurve {keys [<at 0 1>,<at 1 0>]}, YCurve {keys [<at 0 1>,<at 1 0>]}, ZCurve {keys [<at 0 1>,<at 1 0>]}}}
           CurveIndex=&DotProduct(
            Vector A=&ReturnCameraVector(
             Camera Vectors=::NewEnumerator4
             Coordinate Space=::Simulation
             Origin Position=$Particles.Position
             CameraQuery=DataInterface() {propertyValues {bRequireCurrentFrameData=false}}
            )
            Vector B=<0 1 0>
           )
          )
          Channel=::NewEnumerator0
         )
         Input Min=0.2
         Input Max=0.9
         Output Min=1
         Output Max=&MakeCustomFloatFromBool(
          Bool=&Boolean_OrOperation(
           Bool1=&SetBoolByFloatComparison(A=&VectorLength(Vector=$User.OwnerLocalVelocity), B=100)
           Bool2=&SetBoolByFloatComparison(A=$User.FireIntensity, B=0.2)
          )
          True Float=0.65
          False Float=1
         )
        ))
      )
     )
    ))
   &GravityForce(Gravity=<0 -200 0>)
   &Drag(
    Drag=&MakeCustomFloatFromBool(
     Bool=&SetBoolByFloatComparison(A=&VectorLength(Vector=$User.OwnerLocalVelocity), B=200)
     True Float=0
     False Float=10
    )
    Use Linear Drag=true
    Use Rotational Drag=true
   )
   &SolveForcesAndVelocity(Clamp Velocity=false, Limit Acceleration=false)
   &ScaleSpriteSizeBySpeed(Max Scale Factor=<1 40>, Min Scale Factor=<0.01 0.05>, Velocity Threshold=1)
   &ScaleColor(
    Scale Alpha=&Multiply_Float(
     A=&Multiply_Float(
      A=&RandomRangeFloat(Minimum=0.2, Maximum=0.5)
      B=&MakeCustomFloatFromBool(
       Bool=&Boolean_OrOperation(
        Bool1=&SetBoolByFloatComparison(A=&VectorLength(Vector=$User.OwnerLocalVelocity), B=100)
        Bool2=&SetBoolByFloatComparison(A=$User.UseIdleVFX, B=0.2)
       )
       True Float=1
       False Float=0.6
      )
     )
     B=&RemapRange(
      Input Value To Remap=&MakeFloatFromVector(
       Vector=&VectorFromCurve(
        VectorCurve=DataInterface() {propertyValues {XCurve {keys [<at 0 1>,<at 1 0>]}, YCurve {keys [<at 0 1>,<at 1 0>]}, ZCurve {keys [<at 0 1>,<at 1 0>]}}}
        CurveIndex=&DotProduct(
         Vector A=&ReturnCameraVector(
          Camera Vectors=::NewEnumerator4
          Coordinate Space=::Simulation
          Origin Position=$Particles.Position
          CameraQuery=DataInterface() {propertyValues {bRequireCurrentFrameData=false}}
         )
         Vector B=<0 1 0>
        )
       )
       Channel=::NewEnumerator0
      )
      Input Min=0.2
      Input Max=0.9
      Output Min=1
      Output Max=0.6
     )
    )
    Scale RGB=&MakeCustomVectorFromBool(
     Bool=&Boolean_OrOperation(
      Bool1=&SetBoolByFloatComparison(A=&VectorLength(Vector=$User.OwnerLocalVelocity), B=100)
      Bool2=&SetBoolByFloatComparison(A=$User.UseIdleVFX, B=0.2)
     )
     True Vector=<0.25 0.25 0.25>
     False Vector=<20 5 5>
    )
    ScaleA=true
    ScaleRGB=true
   )
   &ScaleColor(
    Scale Alpha=&Multiply_Float(
     A=&RandomRangeFloat(Minimum=0.2, Maximum=0.5)
     B=&MakeCustomFloatFromBool(
      Bool=&Boolean_OrOperation(
       Bool1=&SetBoolByFloatComparison(A=&VectorLength(Vector=$User.OwnerLocalVelocity), B=100)
       Bool2=&SetBoolByFloatComparison(A=$User.UseIdleVFX, B=0.2)
      )
      True Float=1
      False Float=0.6
     )
    )
    Scale RGB=&MakeCustomVectorFromBool(
     Bool=&Boolean_OrOperation(
      Bool1=&SetBoolByFloatComparison(A=&VectorLength(Vector=$User.OwnerLocalVelocity), B=100)
      Bool2=&SetBoolByFloatComparison(A=$User.UseIdleVFX, B=0.2)
     )
     True Vector=<0.25 0.25 0.25>
     False Vector=<20 5 5>
    )
    ScaleA=true
    ScaleRGB=true
   ) {enabled=false, module=ScaleColor001}
  ]
  rendererClasses [&NiagaraSpriteRendererProperties]
  renderers [{rendererClass=&NiagaraSpriteRendererProperties}]
 }
 Thruster_Burst {
  emitterUpdateScript [
   &EmitterState(
    Inactive Response=::NewEnumerator1
    Life Cycle Mode=::NewEnumerator1
    Loop Behavior=::NewEnumerator1
    Loop Duration Mode=::NewEnumerator1
    UseLoopDelay=false
   )
   &SpawnByTrigger(
    Trigger:NiagaraBool=&SetBoolByFloatComparison(A=&VectorLength(Vector=$User.OwnerLocalVelocity), B=20)
    TriggerIntervalSwitch=false
   )
   &SpawnByTrigger(Trigger:NiagaraBool=&SetBoolByFloatComparison(A=$User.FireIntensity, B=0.01), TriggerIntervalSwitch=false) {module=SpawnByTrigger001}
   /SetVariables(
    Emitter.NoiseA_SpeedX:NiagaraFloat=&MakeCustomFloatFromBool(
     Bool=&SetBoolByFloatComparison(A=&VectorLength(Vector=$User.OwnerLocalVelocity), B=200)
     True Float=5
     False Float=2
    )
    Emitter.NoiseA_SpeedY=2
    Emitter.NoiseB_SpeedX:NiagaraFloat=&MakeCustomFloatFromBool(
     Bool=&SetBoolByFloatComparison(A=&VectorLength(Vector=$User.OwnerLocalVelocity), B=200)
     True Float=5
     False Float=2
    )
    Emitter.NoiseB_SpeedY=2
   ) {isSetParameters=true, module=SetVariables_#6}
  ]
  particleSpawnScript [
   &PartitionParticles(Number Of Partitions=3) {enabled=false}
   /SetVariables(Particles.VisibilityTag=&SpawnGroup()) {isSetParameters=true, module=SetVariables_#7}
   &InitializeParticle(
    Color=<rgba 1 1 1 5>
    Lifetime=0.35
    Mass Mode=::NewEnumerator1
    Mesh Scale=&Multiply_VectorByFloat(Vector=<0.2 0.3 0.3>, Float=1)
    Mesh Scale Mode=::NewEnumerator4
    Sprite Size Mode=::NewEnumerator1
    Uniform Sprite Size Max=20.7778
    Uniform Sprite Size Min=34.0823
    UsePositionOffset=false
    Write Material Random=true
   ) {module=InitializeParticle001}
   &DynamicMaterialParameters(
    Index 0 Param 1=1
    Index 0 Param 2=&MakeCustomFloatFromBool(
     Bool=&SetBoolByFloatComparison(A=&VectorLength(Vector=$User.OwnerLocalVelocity), B=200)
     True Float=2
     False Float=1
    )
    Module.Index0Float0Write=true
    Module.Index0Float1Write=true
    Module.Index0Float2Write=true
    Module.Index0Float3Write=true
   )
   &InitialMeshOrientation(Mesh Axis To Orient=<0 2 0>, Mesh Orientation Mode=::NewEnumerator2, Use Rotation Vector=false)
   &CameraOffset(Camera Offset Amount=&MergeHelper_FloatSwitch(0=5, 1=8, 2=0, 3=0, 4=0, Index=&SpawnGroup(), Num=::NewEnumerator3))
   /SetVariables(Particles.GlowFacing:Vector3f=<0 1 0>) {isSetParameters=true, module=SetVariables_#8}
  ]
  particleUpdateScript [
   &ParticleState()
   &ScaleSpriteSize(Uniform Curve Scale=2) {enabled=false}
   &ScaleColor(
    Scale Alpha=&Lerp_Float(A=1, B=0, Alpha=$Particles.NormalizedAge)
    Scale RGB=&MakeVectorFromLinearColorRGB(Color=&ColorFromCurve(ColorCurve=DataInterface() {propertyValues {
        AlphaCurve {keys []}
        BlueCurve {keys [
          <at 0.0626 2>
          <at 0.3286 1>
          {arriveTangent=-1.0668, interpMode=RCIM_Cubic, leaveTangent=-1.0668, time=0.7815, value=1}
          <at 1 1 RCIM_Cubic>
         ]}
        GreenCurve {keys [
          <at 0.0626 2>
          <at 0.3286 1>
          {arriveTangent=-1.7656, interpMode=RCIM_Cubic, leaveTangent=-1.7656, time=0.7815, value=0.4155}
          <at 1 0.3449 RCIM_Cubic>
         ]}
        RedCurve {keys [
          <at 0.0626 2>
          <at 0.3286 1>
          {arriveTangent=-1.8481, interpMode=RCIM_Cubic, leaveTangent=-1.8481, time=0.7815, value=0.2747}
          <at 1 0.2676 RCIM_Cubic>
         ]}
       }}))
    ScaleA=true
    ScaleRGB=true
   )
   &ScaleMeshSize(Scale Factor=&Multiply_VectorByFloat(
     Vector=&MakeVector(
      X=&FloatFromCurve(FloatCurve=DataInterface() {propertyValues {Curve {keys [<at 0 0.75 RCIM_Cubic>, {arriveTangent=-0.25, interpMode=RCIM_Cubic, leaveTangent=-0.25, time=0.5, value=2.5}, <at 1 0.5 RCIM_Cubic>]}}})
      Y=&FloatFromCurve(FloatCurve=DataInterface() {propertyValues {Curve {keys [<at 0.5 1 RCIM_Cubic>,<at 1 1.75 RCIM_Cubic>]}}})
      Z=&FloatFromCurve(FloatCurve=DataInterface() {propertyValues {Curve {keys [<at 0.5 1 RCIM_Cubic>,<at 1 1.75 RCIM_Cubic>]}}})
     )
     Float=&MakeCustomFloatFromBool(
      Bool=&SetBoolByFloatComparison(A=$User.UseIdleVFX, B=1, Comparison Type=::NewEnumerator6)
      True Float=0
      False Float=1
     )
    ))
   &SkeletalMeshLocation(
    Bone / Socket Index=2
    Bone Sampling Mode=::NewEnumerator4
    MaskSpawnGroup=false
    Sampled Position Offset=<0 17 0>
    Skeletal Mesh=$System.Skeletal Mesh Reader
   )
   /SetVariables(
    Particles.Linear Color=&MakeLinearColorFromVectorAndFloat(
     "Vector (RGB)"=&MakeVectorFromLinearColorRGB(Color=&ColorFromCurve(ColorCurve=DataInterface() {propertyValues {
         AlphaCurve {keys []}
         BlueCurve {keys [<at 0.0626 2>,<at 0.3286 1>,<at 0.5683 1 RCIM_Cubic>,<at 1 1 RCIM_Cubic>]}
         GreenCurve {keys [
           <at 0.0626 0.1874>
           <at 0.3286 0.1021>
           {arriveTangent=-0.1521, interpMode=RCIM_Cubic, leaveTangent=-0.1521, time=0.5683, value=0.4155}
           <at 1 0 RCIM_Cubic>
          ]}
         RedCurve {keys [
           <at 0.0626 0.602>
           <at 0.3286 0.7653>
           {arriveTangent=-1.0741, interpMode=RCIM_Cubic, leaveTangent=-1.0741, time=0.5683, value=0.2747}
           <at 1 0.0442 RCIM_Cubic>
          ]}
        }}))
     "Float (Alpha)"=&Lerp_Float(A=0.5, B=0, Alpha=$Particles.NormalizedAge)
    )
    Particles.Vector:Vector3f=&Multiply_VectorByFloat(
     Vector=&Multiply_Vector(A=$Particles.Scale, B=<1.25 2 2>)
     Float=&MakeCustomFloatFromBool(
      Bool=&SetBoolByFloatComparison(A=$User.UseIdleVFX, B=1, Comparison Type=::NewEnumerator6)
      True Float=0
      False Float=1
     )
    )
   ) {isSetParameters=true, module=SetVariables_#9}
  ]
  rendererClasses [&NiagaraMeshRendererProperties]
  renderers [
   {rendererClass=&NiagaraMeshRendererProperties, rendererIndex=0}
   {rendererClass=&NiagaraMeshRendererProperties, rendererIndex=1}
  ]
 }
 Thruster_Loop {
  emitterUpdateScript [
   &EmitterState(
    Inactive Response=::NewEnumerator1
    Life Cycle Mode=::NewEnumerator1
    Loop Behavior=::NewEnumerator1
    Loop Duration Mode=::NewEnumerator1
    UseLoopDelay=false
   )
   &SpawnRate(
    SpawnRate=&MakeCustomFloatFromBool(
     Bool=&SetBoolByFloatComparison(A=&VectorLength(Vector=$User.OwnerLocalVelocity), B=200)
     True Float=50
     False Float=30
    )
    Use Spawn Probability=false
   )
   &SpawnRate(SpawnGroup=1, SpawnRate=10, Use Spawn Probability=false) {module=SpawnRate001}
   &SpawnBurst_Instantaneous(Spawn Group=1, Use Spawn Probability=false, UseLoopCountLimit=false) {enabled=false}
   /SetVariables(
    Emitter.NoiseA_SpeedX:NiagaraFloat=&MakeCustomFloatFromBool(
     Bool=&SetBoolByFloatComparison(A=&VectorLength(Vector=$User.OwnerLocalVelocity), B=200)
     True Float=5
     False Float=2
    )
    Emitter.NoiseA_SpeedY=2
    Emitter.NoiseB_SpeedX:NiagaraFloat=&MakeCustomFloatFromBool(
     Bool=&SetBoolByFloatComparison(A=&VectorLength(Vector=$User.OwnerLocalVelocity), B=200)
     True Float=5
     False Float=2
    )
    Emitter.NoiseB_SpeedY=2
   ) {isSetParameters=true, module=SetVariables_#6}
  ]
  particleSpawnScript [
   /SetVariables(Particles.VisibilityTag=&SpawnGroup()) {isSetParameters=true, module=SetVariables_#7}
   &InitializeParticle(
    Color=<rgba 1 1 1 5>
    Lifetime=0.15
    Mass Mode=::NewEnumerator1
    Mesh Scale=&Multiply_VectorByFloat(
     Vector=<0.08 0.2 0.2>
     Float=&MakeCustomFloatFromBool(
      Bool=&Boolean_OrOperation(
       Bool1=&SetBoolByFloatComparison(A=&VectorLength(Vector=$User.OwnerLocalVelocity), B=100)
       Bool2=&SetBoolByFloatComparison(A=$User.FireIntensity, B=0.1)
      )
      True Float=0.7
      False Float=0.4
     )
    )
    Mesh Scale Mode=::NewEnumerator4
    Sprite Size Mode=::NewEnumerator1
    Uniform Sprite Size Max=20.7778
    Uniform Sprite Size Min=34.0823
    UsePositionOffset=false
    Write Material Random=true
   ) {module=InitializeParticle001}
   &DynamicMaterialParameters(
    Index 0 Param 1=1
    Index 0 Param 2=&MakeCustomFloatFromBool(
     Bool=&SetBoolByFloatComparison(A=&VectorLength(Vector=$User.OwnerLocalVelocity), B=200)
     True Float=2
     False Float=1
    )
    Module.Index0Float0Write=true
    Module.Index0Float1Write=true
    Module.Index0Float2Write=true
    Module.Index0Float3Write=true
   )
   &InitialMeshOrientation(Mesh Axis To Orient=<0 2 0>, Mesh Orientation Mode=::NewEnumerator2, Use Rotation Vector=false)
   &CameraOffset(Camera Offset Amount=&MergeHelper_FloatSwitch(0=5, 1=8, 2=0, 3=0, 4=0, Index=&SpawnGroup(), Num=::NewEnumerator3))
   /SetVariables(Particles.GlowFacing:Vector3f=<0 1 0>) {isSetParameters=true, module=SetVariables_#8}
  ]
  particleUpdateScript [
   &ParticleState()
   &ScaleSpriteSize(Uniform Curve Scale=&Multiply_Float(A=2, B=$User.UseIdleVFX))
   &ScaleColor(
    Scale Alpha=&Lerp_Float(A=0.75, B=0, Alpha=$Particles.NormalizedAge)
    Scale RGB=&MakeCustomVectorFromBool(
     Bool=&Boolean_OrOperation(
      Bool1=&SetBoolByFloatComparison(A=&VectorLength(Vector=$User.OwnerLocalVelocity), B=100)
      Bool2=&SetBoolByFloatComparison(A=$User.FireIntensity, B=0.1)
     )
     True Vector=&MakeVectorFromLinearColorRGB(Color=&ColorFromCurve(ColorCurve=DataInterface() {propertyValues {
         AlphaCurve {keys []}
         BlueCurve {keys [<at 0.0626 1.8937>,<at 0.3286 1>,<at 0.7815 1 RCIM_Cubic>,<at 1 1 RCIM_Cubic>]}
         GreenCurve {keys [
           <at 0.0626 1.3583>
           <at 0.3286 0.5886>
           {arriveTangent=-0.3629, interpMode=RCIM_Cubic, leaveTangent=-0.3629, time=0.7815, value=0.4155}
           <at 1 0.3449 RCIM_Cubic>
          ]}
         RedCurve {keys [
           <at 0.0626 2>
           <at 0.3286 0.8581>
           {arriveTangent=-0.8796, interpMode=RCIM_Cubic, leaveTangent=-0.8796, time=0.7815, value=0.2747}
           <at 1 0.2676 RCIM_Cubic>
          ]}
        }}))
     False Vector=&MakeVectorFromLinearColorRGB(Color=&ColorFromCurve(ColorCurve=DataInterface() {propertyValues {
         AlphaCurve {keys []}
         BlueCurve {keys [<at 0.0755 1>,<at 0.3581 1 RCIM_Cubic>,<at 0.7447 1 RCIM_Cubic>]}
         GreenCurve {keys [
           <at 0.0755 0.6057>
           {arriveTangent=-0.7949, interpMode=RCIM_Cubic, leaveTangent=-0.7949, time=0.3581, value=0.4155}
           <at 0.7447 0.0737 RCIM_Cubic>
          ]}
         RedCurve {keys [
           <at 0.0755 0.7473>
           {arriveTangent=-0.9719, interpMode=RCIM_Cubic, leaveTangent=-0.9719, time=0.3581, value=0.2747}
           <at 0.7447 0.0969 RCIM_Cubic>
          ]}
        }}))
    )
    ScaleA=true
    ScaleRGB=true
   )
   &ScaleMeshSize(Scale Factor=&Multiply_Vector(
     A=&MakeCustomVectorFromBool(
      Bool=&Boolean_OrOperation(
       Bool1=&SetBoolByFloatComparison(A=&VectorLength(Vector=$User.OwnerLocalVelocity), B=100)
       Bool2=&SetBoolByFloatComparison(A=$User.FireIntensity, B=0.1)
      )
      True Vector=<1.5 1 1>
      False Vector=<0.6 1 1>
     )
     B=&Multiply_VectorByFloat(
      Vector=&MakeVector(
       X=&FloatFromCurve(FloatCurve=DataInterface() {propertyValues {Curve {keys [<at 0 0.75 RCIM_Cubic>, {arriveTangent=-0.25, interpMode=RCIM_Cubic, leaveTangent=-0.25, time=0.5, value=2.5}, <at 1 0.5 RCIM_Cubic>]}}})
       Y=&FloatFromCurve(FloatCurve=DataInterface() {propertyValues {Curve {keys [<at 0.5 1 RCIM_Cubic>,<at 1 1.75 RCIM_Cubic>]}}})
       Z=&FloatFromCurve(FloatCurve=DataInterface() {propertyValues {Curve {keys [<at 0.5 1 RCIM_Cubic>,<at 1 1.75 RCIM_Cubic>]}}})
      )
      Float=$User.UseIdleVFX
     )
    ))
   /SetVariables(Particles.Scale_2:Vector3f=&Multiply_Vector(
     A=$Particles.Initial.Scale_2
     B=&MakeVector(
      X=&FloatFromCurve(FloatCurve=DataInterface() {propertyValues {Curve {keys [<at 0 0.75 RCIM_Cubic>, {arriveTangent=-0.25, interpMode=RCIM_Cubic, leaveTangent=-0.25, time=0.5, value=2.5}, <at 1 0.5 RCIM_Cubic>]}}})
      Y=&FloatFromCurve(FloatCurve=DataInterface() {propertyValues {Curve {keys [<at 0.5 1 RCIM_Cubic>,<at 1 1.75 RCIM_Cubic>]}}})
      Z=&FloatFromCurve(FloatCurve=DataInterface() {propertyValues {Curve {keys [<at 0.5 1 RCIM_Cubic>,<at 1 1.75 RCIM_Cubic>]}}})
     )
    )) {isSetParameters=true, module=SetVariables_#10}
   &SkeletalMeshLocation(
    Bone / Socket Index=2
    Bone Sampling Mode=::NewEnumerator4
    MaskSpawnGroup=false
    Sampled Position Offset=<0 17 0>
    Skeletal Mesh=$System.Skeletal Mesh Reader
   )
   /SetVariables(
    Particles.Linear Color=&MakeLinearColorFromVectorAndFloat(
     "Vector (RGB)"=&MakeVectorFromLinearColorRGB(Color=&ColorFromCurve(ColorCurve=DataInterface() {propertyValues {
         AlphaCurve {keys []}
         BlueCurve {keys [<at 0.0626 2>,<at 0.3286 1>,<at 0.5683 1 RCIM_Cubic>,<at 1 1 RCIM_Cubic>]}
         GreenCurve {keys [
           <at 0.0626 0.1874>
           <at 0.3286 0.1021>
           {arriveTangent=-0.1521, interpMode=RCIM_Cubic, leaveTangent=-0.1521, time=0.5683, value=0.4155}
           <at 1 0 RCIM_Cubic>
          ]}
         RedCurve {keys [
           <at 0.0626 0.602>
           <at 0.3286 0.7653>
           {arriveTangent=-1.0741, interpMode=RCIM_Cubic, leaveTangent=-1.0741, time=0.5683, value=0.2747}
           <at 1 0.0442 RCIM_Cubic>
          ]}
        }}))
     "Float (Alpha)"=&Lerp_Float(A=0.5, B=0, Alpha=$Particles.NormalizedAge)
    )
    Particles.Vector:Vector3f=&Multiply_Vector(
     A=&MakeCustomVectorFromBool(
      Bool=&Boolean_OrOperation(
       Bool1=&SetBoolByFloatComparison(A=&VectorLength(Vector=$User.OwnerLocalVelocity), B=100)
       Bool2=&SetBoolByFloatComparison(A=$User.FireIntensity, B=0.1)
      )
      True Vector=<1 1 1>
      False Vector=<0.6 1 1>
     )
     B=&Multiply_VectorByFloat(Vector=&Multiply_Vector(A=$Particles.Scale, B=<1.25 2 2>), Float=$User.UseIdleVFX)
    )
   ) {isSetParameters=true, module=SetVariables_#9}
   /SetVariables(
    Particles.Flare Color:LinearColor=&MakeCustomLinearColorFromBool(
     Bool=&Boolean_OrOperation(
      Bool1=&SetBoolByFloatComparison(A=&VectorLength(Vector=$User.OwnerLocalVelocity), B=100)
      Bool2=&SetBoolByFloatComparison(A=$User.FireIntensity, B=0.1)
     )
     True Color=<rgba 3 1.4133 2.6033 0.8>
     False Color=<rgba 1.7049 0.8346 3 0.75>
    )
    Particles.Flare Sprite Size:Vector2f=&Multiply_Vector2D_ByFloat(
     Float=$User.UseIdleVFX
     Vector2D=&Vector2DFromFloat(Value=&Multiply_Float(
       A=&MakeCustomFloatFromBool(
        Bool=&SetBoolByFloatComparison(A=&VectorLength(Vector=$User.OwnerLocalVelocity), B=100)
        True Float=100
        False Float=85
       )
       B=&Waveform(
        "Waveform [1]"=::NewEnumerator6
        "[1] Frequency"=&MakeCustomFloatFromBool(
         Bool=&Boolean_OrOperation(
          Bool1=&SetBoolByFloatComparison(A=&VectorLength(Vector=$User.OwnerLocalVelocity), B=100)
          Bool2=&SetBoolByFloatComparison(A=$User.UseIdleVFX, B=0.01)
         )
         True Float=5
         False Float=3
        )
        "[1] Amplitude Min/Max"=<0.35 1>
       )
      ))
    )
   ) {isSetParameters=true, module=SetVariables_#11}
   /SetVariables(
    Particles.Glow Color=<rgba 3 1.4133 2.6033 0.8>
    Particles.Glow Sprite Size:Vector2f=&Multiply_Vector2D_ByFloat(
     Float=$User.UseIdleVFX
     Vector2D=&Vector2DFromFloat(Value=&Multiply_Float(
       A=&MakeCustomFloatFromBool(
        Bool=&SetBoolByFloatComparison(A=&VectorLength(Vector=$User.OwnerLocalVelocity), B=100)
        True Float=35
        False Float=20
       )
       B=&Waveform(
        "Waveform [1]"=::NewEnumerator6
        "[1] Frequency"=&MakeCustomFloatFromBool(
         Bool=&Boolean_OrOperation(
          Bool1=&SetBoolByFloatComparison(A=&VectorLength(Vector=$User.OwnerLocalVelocity), B=100)
          Bool2=&SetBoolByFloatComparison(A=$User.FireIntensity, B=0.1)
         )
         True Float=5
         False Float=3
        )
        "[1] Amplitude Min/Max"=<0.35 1>
       )
      ))
    )
   ) {isSetParameters=true, module=SetVariables_#12}
   /SetVariables(
    Particles.Outer Glow Color:LinearColor=<rgba 0.0104 0.0737 1 0.367>
    Particles.Outer Glow Sprite Size:Vector2f=&Multiply_Vector2D(
     A=&MakeCustomVector2DFromBool(
      Bool=&Boolean_OrOperation(
       Bool1=&SetBoolByFloatComparison(A=&VectorLength(Vector=$User.OwnerLocalVelocity), B=100)
       Bool2=&SetBoolByFloatComparison(A=$User.FireIntensity, B=0.1)
      )
      True Vector2D=<1 1>
      False Vector2D=<1 1>
     )
     B=&Multiply_Vector2D_ByFloat(
      Float=$User.UseIdleVFX
      Vector2D=&Vector2DFromFloat(Value=&Multiply_Float(
        A=&MakeCustomFloatFromBool(
         Bool=&Boolean_OrOperation(
          Bool1=&SetBoolByFloatComparison(A=&VectorLength(Vector=$User.OwnerLocalVelocity), B=100)
          Bool2=&SetBoolByFloatComparison(A=$User.FireIntensity, B=0.1)
         )
         True Float=85
         False Float=50
        )
        B=&Waveform(
         "Waveform [1]"=::NewEnumerator6
         "[1] Frequency"=&MakeCustomFloatFromBool(
          Bool=&Boolean_OrOperation(
           Bool1=&SetBoolByFloatComparison(A=&VectorLength(Vector=$User.OwnerLocalVelocity), B=100)
           Bool2=&SetBoolByFloatComparison(A=$User.FireIntensity, B=0.1)
          )
          True Float=5
          False Float=3
         )
         "[1] Amplitude Min/Max"=<0.35 1>
        )
       ))
     )
    )
   ) {isSetParameters=true, module=SetVariables_#13}
   /SetVariables(Particles.GlowsPos:NiagaraPosition=&AddVectorToPosition(
     Position=$Particles.Position
     Vector=&MakeCustomVectorFromBool(
      Bool=&Boolean_OrOperation(
       Bool1=&SetBoolByFloatComparison(A=&VectorLength(Vector=$User.OwnerLocalVelocity), B=100)
       Bool2=&SetBoolByFloatComparison(A=$User.FireIntensity, B=0.1)
      )
      True Vector=<0 20 0>
      False Vector=<0 5 0>
     )
    )) {isSetParameters=true, module=SetVariables_#14}
   /SetVariables(
    Particles.Rotate Color:LinearColor=&MakeLinearColorFromVectorAndFloat(
     "Vector (RGB)"=&MakeCustomVectorFromBool(
      Bool=&Boolean_OrOperation(
       Bool1=&SetBoolByFloatComparison(A=&VectorLength(Vector=$User.OwnerLocalVelocity), B=100)
       Bool2=&SetBoolByFloatComparison(A=$User.FireIntensity, B=0.1)
      )
      True Vector=&MakeVectorFromLinearColorRGB(Color=&ColorFromCurve(ColorCurve=DataInterface() {propertyValues {
          AlphaCurve {keys []}
          BlueCurve {keys [<at 0.0688 3>,<at 0.2522 1>,<at 0.4086 1>,<at 0.9685 1 RCIM_Cubic>]}
          GreenCurve {keys [<at 0.0688 3>,<at 0.2522 1>,<at 0.4086 1>,<at 0.9685 0.761 RCIM_Cubic>]}
          RedCurve {keys [<at 0.0688 3>,<at 0.2522 1>,<at 0.4086 1>,<at 0.9685 0.9015 RCIM_Cubic>]}
         }}))
      False Vector=&MakeVectorFromLinearColorRGB(Color=&ColorFromCurve(ColorCurve=DataInterface() {propertyValues {
          AlphaCurve {keys []}
          BlueCurve {keys [<at 0.0939 1.5 RCIM_Cubic>,<at 0.9569 1 RCIM_Cubic>]}
          GreenCurve {keys [<at 0.0939 1.2705 RCIM_Cubic>,<at 0.9569 0.5035 RCIM_Cubic>]}
          RedCurve {keys [<at 0.0939 1.342 RCIM_Cubic>,<at 0.9569 0.7039 RCIM_Cubic>]}
         }}))
     )
     "Float (Alpha)"=&Lerp_Float(A=1, B=0, Alpha=$Particles.NormalizedAge)
    )
    Particles.Rotate Size:Vector3f=&Multiply_Vector(
     A=&MakeCustomVectorFromBool(
      Bool=&Boolean_OrOperation(
       Bool1=&SetBoolByFloatComparison(A=&VectorLength(Vector=$User.OwnerLocalVelocity), B=100)
       Bool2=&SetBoolByFloatComparison(A=$User.FireIntensity, B=0.1)
      )
      True Vector=<1 1 1>
      False Vector=<0.6 1 1>
     )
     B=&Multiply_VectorByFloat(Vector=&Multiply_Vector(A=$Particles.Scale, B=<1 1 1>), Float=$User.UseIdleVFX)
    )
   ) {isSetParameters=true, module=SetVariables_#15}
  ]
  rendererClasses [&NiagaraMeshRendererProperties, &NiagaraSpriteRendererProperties]
  renderers [
   {rendererClass=&NiagaraMeshRendererProperties, rendererIndex=0}
   {rendererClass=&NiagaraMeshRendererProperties, rendererIndex=1}
   {rendererClass=&NiagaraSpriteRendererProperties, rendererIndex=2}
   {rendererClass=&NiagaraSpriteRendererProperties, rendererIndex=3}
   {rendererClass=&NiagaraSpriteRendererProperties, rendererIndex=4}
   {rendererClass=&NiagaraMeshRendererProperties, rendererIndex=5}
  ]
 }
 Thruster_Loop001 {
  simTarget=GPUComputeSim
  emitterUpdateScript [
   &EmitterState(
    Inactive Response=::NewEnumerator1
    Life Cycle Mode=::NewEnumerator1
    Loop Behavior=::NewEnumerator1
    Loop Duration Mode=::NewEnumerator1
    UseLoopDelay=false
   )
   &SpawnRate(
    SpawnRate=&MakeCustomFloatFromBool(
     Bool=&SetBoolByFloatComparison(A=&VectorLength(Vector=$User.OwnerLocalVelocity), B=200)
     True Float=50
     False Float=30
    )
    Use Spawn Probability=false
   )
   &SpawnRate(SpawnGroup=1, SpawnRate=10, Use Spawn Probability=false) {module=SpawnRate001}
   &SpawnBurst_Instantaneous(Spawn Group=1, Use Spawn Probability=false, UseLoopCountLimit=false) {enabled=false}
   /SetVariables(
    Emitter.NoiseA_SpeedX:NiagaraFloat=&MakeCustomFloatFromBool(
     Bool=&SetBoolByFloatComparison(A=&VectorLength(Vector=$User.OwnerLocalVelocity), B=200)
     True Float=5
     False Float=2
    )
    Emitter.NoiseA_SpeedY=2
    Emitter.NoiseB_SpeedX:NiagaraFloat=&MakeCustomFloatFromBool(
     Bool=&SetBoolByFloatComparison(A=&VectorLength(Vector=$User.OwnerLocalVelocity), B=200)
     True Float=5
     False Float=2
    )
    Emitter.NoiseB_SpeedY=2
   ) {isSetParameters=true, module=SetVariables_#6}
  ]
  particleSpawnScript [
   /SetVariables(Particles.VisibilityTag=&SpawnGroup()) {isSetParameters=true, module=SetVariables_#7}
   &InitializeParticle(
    Color=<rgba 1 1 1 5>
    Lifetime=0.15
    Mass Mode=::NewEnumerator1
    Mesh Scale=&Multiply_VectorByFloat(
     Vector=<0.08 0.2 0.2>
     Float=&MakeCustomFloatFromBool(
      Bool=&Boolean_OrOperation(
       Bool1=&SetBoolByFloatComparison(A=&VectorLength(Vector=$User.OwnerLocalVelocity), B=100)
       Bool2=&SetBoolByFloatComparison(A=$User.FireIntensity, B=0.1)
      )
      True Float=0.7
      False Float=0.4
     )
    )
    Mesh Scale Mode=::NewEnumerator4
    Sprite Size Mode=::NewEnumerator1
    Uniform Sprite Size Max=20.7778
    Uniform Sprite Size Min=34.0823
    UsePositionOffset=false
    Write Material Random=true
   ) {module=InitializeParticle001}
   &DynamicMaterialParameters(
    Index 0 Param 1=1
    Index 0 Param 2=&MakeCustomFloatFromBool(
     Bool=&SetBoolByFloatComparison(A=&VectorLength(Vector=$User.OwnerLocalVelocity), B=200)
     True Float=2
     False Float=1
    )
    Module.Index0Float0Write=true
    Module.Index0Float1Write=true
    Module.Index0Float2Write=true
    Module.Index0Float3Write=true
   )
   &InitialMeshOrientation(Mesh Axis To Orient=<0 2 0>, Mesh Orientation Mode=::NewEnumerator2, Use Rotation Vector=false)
   &CameraOffset(Camera Offset Amount=&MergeHelper_FloatSwitch(0=5, 1=8, 2=0, 3=0, 4=0, Index=&SpawnGroup(), Num=::NewEnumerator3))
   /SetVariables(Particles.GlowFacing:Vector3f=<0 1 0>) {isSetParameters=true, module=SetVariables_#8}
  ]
  particleUpdateScript [
   &ParticleState()
   &ScaleSpriteSize(Uniform Curve Scale=&Multiply_Float(A=2, B=$User.UseIdleVFX))
   &ScaleColor(
    Scale Alpha=&Lerp_Float(A=0.75, B=0, Alpha=$Particles.NormalizedAge)
    Scale RGB=&MakeCustomVectorFromBool(
     Bool=&Boolean_OrOperation(
      Bool1=&SetBoolByFloatComparison(A=&VectorLength(Vector=$User.OwnerLocalVelocity), B=100)
      Bool2=&SetBoolByFloatComparison(A=$User.FireIntensity, B=0.1)
     )
     True Vector=&MakeVectorFromLinearColorRGB(Color=&ColorFromCurve(ColorCurve=DataInterface() {propertyValues {
         AlphaCurve {keys []}
         BlueCurve {keys [<at 0.0626 1.8937>,<at 0.3286 1>,<at 0.7815 1 RCIM_Cubic>,<at 1 1 RCIM_Cubic>]}
         GreenCurve {keys [
           <at 0.0626 1.3583>
           <at 0.3286 0.5886>
           {arriveTangent=-0.3629, interpMode=RCIM_Cubic, leaveTangent=-0.3629, time=0.7815, value=0.4155}
           <at 1 0.3449 RCIM_Cubic>
          ]}
         RedCurve {keys [
           <at 0.0626 2>
           <at 0.3286 0.8581>
           {arriveTangent=-0.8796, interpMode=RCIM_Cubic, leaveTangent=-0.8796, time=0.7815, value=0.2747}
           <at 1 0.2676 RCIM_Cubic>
          ]}
        }}))
     False Vector=&MakeVectorFromLinearColorRGB(Color=&ColorFromCurve(ColorCurve=DataInterface() {propertyValues {
         AlphaCurve {keys []}
         BlueCurve {keys [<at 0.0755 1>,<at 0.3581 1 RCIM_Cubic>,<at 0.7447 1 RCIM_Cubic>]}
         GreenCurve {keys [
           <at 0.0755 0.6057>
           {arriveTangent=-0.7949, interpMode=RCIM_Cubic, leaveTangent=-0.7949, time=0.3581, value=0.4155}
           <at 0.7447 0.0737 RCIM_Cubic>
          ]}
         RedCurve {keys [
           <at 0.0755 0.7473>
           {arriveTangent=-0.9719, interpMode=RCIM_Cubic, leaveTangent=-0.9719, time=0.3581, value=0.2747}
           <at 0.7447 0.0969 RCIM_Cubic>
          ]}
        }}))
    )
    ScaleA=true
    ScaleRGB=true
   )
   &ScaleMeshSize(Scale Factor=&Multiply_Vector(
     A=&MakeCustomVectorFromBool(
      Bool=&Boolean_OrOperation(
       Bool1=&SetBoolByFloatComparison(A=&VectorLength(Vector=$User.OwnerLocalVelocity), B=100)
       Bool2=&SetBoolByFloatComparison(A=$User.FireIntensity, B=0.1)
      )
      True Vector=<1.5 1 1>
      False Vector=<0.6 1 1>
     )
     B=&Multiply_VectorByFloat(
      Vector=&MakeVector(
       X=&FloatFromCurve(FloatCurve=DataInterface() {propertyValues {Curve {keys [<at 0 0.75 RCIM_Cubic>, {arriveTangent=-0.25, interpMode=RCIM_Cubic, leaveTangent=-0.25, time=0.5, value=2.5}, <at 1 0.5 RCIM_Cubic>]}}})
       Y=&FloatFromCurve(FloatCurve=DataInterface() {propertyValues {Curve {keys [<at 0.5 1 RCIM_Cubic>,<at 1 1.75 RCIM_Cubic>]}}})
       Z=&FloatFromCurve(FloatCurve=DataInterface() {propertyValues {Curve {keys [<at 0.5 1 RCIM_Cubic>,<at 1 1.75 RCIM_Cubic>]}}})
      )
      Float=$User.UseIdleVFX
     )
    ))
   /SetVariables(Particles.Scale_2:Vector3f=&Multiply_Vector(
     A=$Particles.Initial.Scale_2
     B=&MakeVector(
      X=&FloatFromCurve(FloatCurve=DataInterface() {propertyValues {Curve {keys [<at 0 0.75 RCIM_Cubic>, {arriveTangent=-0.25, interpMode=RCIM_Cubic, leaveTangent=-0.25, time=0.5, value=2.5}, <at 1 0.5 RCIM_Cubic>]}}})
      Y=&FloatFromCurve(FloatCurve=DataInterface() {propertyValues {Curve {keys [<at 0.5 1 RCIM_Cubic>,<at 1 1.75 RCIM_Cubic>]}}})
      Z=&FloatFromCurve(FloatCurve=DataInterface() {propertyValues {Curve {keys [<at 0.5 1 RCIM_Cubic>,<at 1 1.75 RCIM_Cubic>]}}})
     )
    )) {isSetParameters=true, module=SetVariables_#10}
   &SkeletalMeshLocation(
    Bone / Socket Index=2
    Bone Sampling Mode=::NewEnumerator4
    MaskSpawnGroup=false
    Sampled Position Offset=<0 17 0>
    Skeletal Mesh=$System.Skeletal Mesh Reader
   )
   /SetVariables(
    Particles.Linear Color=&MakeLinearColorFromVectorAndFloat(
     "Vector (RGB)"=&MakeVectorFromLinearColorRGB(Color=&ColorFromCurve(ColorCurve=DataInterface() {propertyValues {
         AlphaCurve {keys []}
         BlueCurve {keys [<at 0.0626 2>,<at 0.3286 1>,<at 0.5683 1 RCIM_Cubic>,<at 1 1 RCIM_Cubic>]}
         GreenCurve {keys [
           <at 0.0626 0.1874>
           <at 0.3286 0.1021>
           {arriveTangent=-0.1521, interpMode=RCIM_Cubic, leaveTangent=-0.1521, time=0.5683, value=0.4155}
           <at 1 0 RCIM_Cubic>
          ]}
         RedCurve {keys [
           <at 0.0626 0.602>
           <at 0.3286 0.7653>
           {arriveTangent=-1.0741, interpMode=RCIM_Cubic, leaveTangent=-1.0741, time=0.5683, value=0.2747}
           <at 1 0.0442 RCIM_Cubic>
          ]}
        }}))
     "Float (Alpha)"=&Lerp_Float(A=0.5, B=0, Alpha=$Particles.NormalizedAge)
    )
    Particles.Vector:Vector3f=&Multiply_Vector(
     A=&MakeCustomVectorFromBool(
      Bool=&Boolean_OrOperation(
       Bool1=&SetBoolByFloatComparison(A=&VectorLength(Vector=$User.OwnerLocalVelocity), B=100)
       Bool2=&SetBoolByFloatComparison(A=$User.FireIntensity, B=0.1)
      )
      True Vector=<1 1 1>
      False Vector=<0.6 1 1>
     )
     B=&Multiply_VectorByFloat(Vector=&Multiply_Vector(A=$Particles.Scale, B=<1.25 2 2>), Float=$User.UseIdleVFX)
    )
   ) {isSetParameters=true, module=SetVariables_#9}
   /SetVariables(
    Particles.Flare Color:LinearColor=&MakeCustomLinearColorFromBool(
     Bool=&Boolean_OrOperation(
      Bool1=&SetBoolByFloatComparison(A=&VectorLength(Vector=$User.OwnerLocalVelocity), B=100)
      Bool2=&SetBoolByFloatComparison(A=$User.FireIntensity, B=0.1)
     )
     True Color=<rgba 3 1.4133 2.6033 0.8>
     False Color=<rgba 1.7049 0.8346 3 0.75>
    )
    Particles.Flare Sprite Size:Vector2f=&Multiply_Vector2D_ByFloat(
     Float=$User.UseIdleVFX
     Vector2D=&Vector2DFromFloat(Value=&Multiply_Float(
       A=&MakeCustomFloatFromBool(
        Bool=&SetBoolByFloatComparison(A=&VectorLength(Vector=$User.OwnerLocalVelocity), B=100)
        True Float=100
        False Float=85
       )
       B=&Waveform(
        "Waveform [1]"=::NewEnumerator6
        "[1] Frequency"=&MakeCustomFloatFromBool(
         Bool=&Boolean_OrOperation(
          Bool1=&SetBoolByFloatComparison(A=&VectorLength(Vector=$User.OwnerLocalVelocity), B=100)
          Bool2=&SetBoolByFloatComparison(A=$User.UseIdleVFX, B=0.01)
         )
         True Float=5
         False Float=3
        )
        "[1] Amplitude Min/Max"=<0.35 1>
       )
      ))
    )
   ) {isSetParameters=true, module=SetVariables_#11}
   /SetVariables(
    Particles.Glow Color=<rgba 3 1.4133 2.6033 0.8>
    Particles.Glow Sprite Size:Vector2f=&Multiply_Vector2D_ByFloat(
     Float=$User.UseIdleVFX
     Vector2D=&Vector2DFromFloat(Value=&Multiply_Float(
       A=&MakeCustomFloatFromBool(
        Bool=&SetBoolByFloatComparison(A=&VectorLength(Vector=$User.OwnerLocalVelocity), B=100)
        True Float=35
        False Float=20
       )
       B=&Waveform(
        "Waveform [1]"=::NewEnumerator6
        "[1] Frequency"=&MakeCustomFloatFromBool(
         Bool=&Boolean_OrOperation(
          Bool1=&SetBoolByFloatComparison(A=&VectorLength(Vector=$User.OwnerLocalVelocity), B=100)
          Bool2=&SetBoolByFloatComparison(A=$User.FireIntensity, B=0.1)
         )
         True Float=5
         False Float=3
        )
        "[1] Amplitude Min/Max"=<0.35 1>
       )
      ))
    )
   ) {isSetParameters=true, module=SetVariables_#12}
   /SetVariables(
    Particles.Outer Glow Color:LinearColor=<rgba 0.0104 0.0737 1 0.367>
    Particles.Outer Glow Sprite Size:Vector2f=&Multiply_Vector2D(
     A=&MakeCustomVector2DFromBool(
      Bool=&Boolean_OrOperation(
       Bool1=&SetBoolByFloatComparison(A=&VectorLength(Vector=$User.OwnerLocalVelocity), B=100)
       Bool2=&SetBoolByFloatComparison(A=$User.FireIntensity, B=0.1)
      )
      True Vector2D=<1 1>
      False Vector2D=<1 1>
     )
     B=&Multiply_Vector2D_ByFloat(
      Float=$User.UseIdleVFX
      Vector2D=&Vector2DFromFloat(Value=&Multiply_Float(
        A=&MakeCustomFloatFromBool(
         Bool=&Boolean_OrOperation(
          Bool1=&SetBoolByFloatComparison(A=&VectorLength(Vector=$User.OwnerLocalVelocity), B=100)
          Bool2=&SetBoolByFloatComparison(A=$User.FireIntensity, B=0.1)
         )
         True Float=85
         False Float=50
        )
        B=&Waveform(
         "Waveform [1]"=::NewEnumerator6
         "[1] Frequency"=&MakeCustomFloatFromBool(
          Bool=&Boolean_OrOperation(
           Bool1=&SetBoolByFloatComparison(A=&VectorLength(Vector=$User.OwnerLocalVelocity), B=100)
           Bool2=&SetBoolByFloatComparison(A=$User.FireIntensity, B=0.1)
          )
          True Float=5
          False Float=3
         )
         "[1] Amplitude Min/Max"=<0.35 1>
        )
       ))
     )
    )
   ) {isSetParameters=true, module=SetVariables_#13}
   /SetVariables(Particles.GlowsPos:NiagaraPosition=&AddVectorToPosition(
     Position=$Particles.Position
     Vector=&MakeCustomVectorFromBool(
      Bool=&Boolean_OrOperation(
       Bool1=&SetBoolByFloatComparison(A=&VectorLength(Vector=$User.OwnerLocalVelocity), B=100)
       Bool2=&SetBoolByFloatComparison(A=$User.FireIntensity, B=0.1)
      )
      True Vector=<0 20 0>
      False Vector=<0 5 0>
     )
    )) {isSetParameters=true, module=SetVariables_#14}
   /SetVariables(
    Particles.Rotate Color:LinearColor=&MakeLinearColorFromVectorAndFloat(
     "Vector (RGB)"=&MakeCustomVectorFromBool(
      Bool=&Boolean_OrOperation(
       Bool1=&SetBoolByFloatComparison(A=&VectorLength(Vector=$User.OwnerLocalVelocity), B=100)
       Bool2=&SetBoolByFloatComparison(A=$User.FireIntensity, B=0.1)
      )
      True Vector=&MakeVectorFromLinearColorRGB(Color=&ColorFromCurve(ColorCurve=DataInterface() {propertyValues {
          AlphaCurve {keys []}
          BlueCurve {keys [<at 0.0688 3>,<at 0.2522 1>,<at 0.4086 1>,<at 0.9685 1 RCIM_Cubic>]}
          GreenCurve {keys [<at 0.0688 3>,<at 0.2522 1>,<at 0.4086 1>,<at 0.9685 0.761 RCIM_Cubic>]}
          RedCurve {keys [<at 0.0688 3>,<at 0.2522 1>,<at 0.4086 1>,<at 0.9685 0.9015 RCIM_Cubic>]}
         }}))
      False Vector=&MakeVectorFromLinearColorRGB(Color=&ColorFromCurve(ColorCurve=DataInterface() {propertyValues {
          AlphaCurve {keys []}
          BlueCurve {keys [<at 0.0939 1.5 RCIM_Cubic>,<at 0.9569 1 RCIM_Cubic>]}
          GreenCurve {keys [<at 0.0939 1.2705 RCIM_Cubic>,<at 0.9569 0.5035 RCIM_Cubic>]}
          RedCurve {keys [<at 0.0939 1.342 RCIM_Cubic>,<at 0.9569 0.7039 RCIM_Cubic>]}
         }}))
     )
     "Float (Alpha)"=&Lerp_Float(A=1, B=0, Alpha=$Particles.NormalizedAge)
    )
    Particles.Rotate Size:Vector3f=&Multiply_Vector(
     A=&MakeCustomVectorFromBool(
      Bool=&Boolean_OrOperation(
       Bool1=&SetBoolByFloatComparison(A=&VectorLength(Vector=$User.OwnerLocalVelocity), B=100)
       Bool2=&SetBoolByFloatComparison(A=$User.FireIntensity, B=0.1)
      )
      True Vector=<1 1 1>
      False Vector=<0.6 1 1>
     )
     B=&Multiply_VectorByFloat(Vector=&Multiply_Vector(A=$Particles.Scale, B=<1 1 1>), Float=$User.UseIdleVFX)
    )
   ) {isSetParameters=true, module=SetVariables_#15}
  ]
  rendererClasses [&NiagaraMeshRendererProperties, &NiagaraSpriteRendererProperties]
  renderers [
   {rendererClass=&NiagaraMeshRendererProperties, rendererIndex=0}
   {rendererClass=&NiagaraMeshRendererProperties, rendererIndex=1}
   {rendererClass=&NiagaraSpriteRendererProperties, rendererIndex=2}
   {rendererClass=&NiagaraSpriteRendererProperties, rendererIndex=3}
   {rendererClass=&NiagaraSpriteRendererProperties, rendererIndex=4}
   {rendererClass=&NiagaraMeshRendererProperties, rendererIndex=5}
  ]
 }
 WindBurstCondensed {
  emitterUpdateScript [
   &EmitterState(
    Inactive Response=::NewEnumerator1
    Life Cycle Mode=::NewEnumerator1
    Loop Behavior=::NewEnumerator1
    Loop Duration Mode=::NewEnumerator1
    UseLoopDelay=false
   )
   &SpawnByTrigger(
    Trigger:NiagaraBool=&SetBoolByFloatComparison(A=&VectorLength(Vector=$User.OwnerLocalVelocity), B=100, Comparison Type=::NewEnumerator2)
    TriggerIntervalSwitch=false
   )
   &SpawnByTrigger(Trigger:NiagaraBool=&SetBoolByFloatComparison(A=$User.FireIntensity, B=0), TriggerIntervalSwitch=false) {module=SpawnByTrigger001}
   &SpawnByTrigger(
    Spawn Group=1
    Trigger:NiagaraBool=&SetBoolByFloatComparison(A=$User.FireIntensity, B=0.25)
    TriggerIntervalSwitch=false
   ) {module=SpawnByTrigger002}
  ]
  particleSpawnScript [
   /SetVariables(Particles.VisibilityTag=&SpawnGroup()) {isSetParameters=true}
   &InitializeParticle(
    Color=&MaskLinearColor(
     Mask Source=0
     Mask Parameter=?
     Mask=0
     Masked Linear Color=&FNTimeOfDayColor(
      Night Color=<rgba 0.3116 0.3916 1 3>
      Morning Color=<rgba 0.8974 0.4935 0.4987 2>
      Day Color=<rgba 1 1 1 1.55>
      Evening Color=<rgba 0.8453 0.5816 0.8974 2>
     )
     Passthrough Linear Color=&FN_Select_Color_from_Time_Of_Day(
      Morning=<rgba 0.7548 0.6128 0.6114 0.13>
      Day=<rgba 0.7548 0.7548 0.7548 0.13>
      Evening=<rgba 0.7548 0.6097 0.6552 0.13>
      Night=<rgba 0.5709 0.6067 0.7548 0.13>
     )
    )
    Lifetime=&MaskFloat(Mask Source=0, Mask Parameter=?, Mask=0, Masked Float=3, Passthrough Float=0.5)
    Mesh Scale=<1.8 1.8 1.8>
    Mesh Scale Mode=::NewEnumerator4
    Position Offset=&MaskVector(Mask Source=0, Mask Parameter=?, Mask=0, Masked Vector=<0 -85 130>, Passthrough Vector=<0 -85 150>)
    Sprite Size Mode=::NewEnumerator3
    Uniform Sprite Size=800
    UsePositionOffset=true
    Write Material Random=true
   )
   &InitialMeshOrientation(Mesh Orientation Mode=::NewEnumerator2, Orientation Vector=<-1 0 0>, Rotation=<0.11 -0.02 0>, Use Rotation Vector=true)
   &CameraOffset(Camera Offset Amount=&MaskFloat(Mask Source=0, Mask Parameter=?, Mask=0, Masked Float=0, Passthrough Float=-20))
  ]
  particleUpdateScript [
   &ParticleState()
   &ScaleMeshSize(Scale Factor=&Multiply_VectorByFloat(Vector=<1 1 1>, Float=$User.UseIdleVFX))
   &ScaleColor(
    Scale Alpha=&MaskFloat(
     Mask Source=0
     Mask Parameter=?
     Mask=0
     Masked Float=&FloatFromCurve(
      FloatCurve=DataInterface() {propertyValues {Curve {
         defaultValue=0
         keys [
          {arriveTangent=0.835, interpMode=RCIM_Cubic, leaveTangent=0.835, tangentMode=RCTM_User, time=0, value=0}
          {arriveTangent=-0.6546, interpMode=RCIM_Cubic, leaveTangent=-0.6546, tangentMode=RCTM_User, time=0.1374, value=0.9754}
          {arriveTangent=-0.3567, interpMode=RCIM_Cubic, leaveTangent=-0.3567, tangentMode=RCTM_User, time=0.1901, value=0.6292}
          {arriveTangent=-0.7768, interpMode=RCIM_Cubic, leaveTangent=-0.7768, time=0.4582, value=0.5339}
          {arriveTangent=-0.0405, interpMode=RCIM_Cubic, leaveTangent=-0.0405, tangentMode=RCTM_User, time=1, value=0}
         ]
        }}}
      Scale Curve=1.2
     )
     Passthrough Float=&FloatFromCurve(
      FloatCurve=DataInterface() {propertyValues {Curve {
         defaultValue=0.25
         keys [
          {arriveTangent=-3.5471, interpMode=RCIM_Cubic, leaveTangent=-3.5471, tangentMode=RCTM_User, time=0.0252, value=1}
          {arriveTangent=-1.0258, interpMode=RCIM_Cubic, leaveTangent=-1.0258, time=0.2244, value=0.385}
          {interpMode=RCIM_Cubic, tangentMode=RCTM_User, time=1, value=0}
         ]
        }}}
      Scale Curve=0.4
     )
    )
    ScaleA=true
    ScaleRGB=false
   ) {module=ScaleColor001}
  ]
  rendererClasses [&NiagaraMeshRendererProperties, &NiagaraSpriteRendererProperties]
  renderers [
   {rendererClass=&NiagaraMeshRendererProperties, rendererIndex=0}
   {rendererClass=&NiagaraSpriteRendererProperties, rendererIndex=1}
  ]
 }
 WindIdle {
  emitterUpdateScript [
   &EmitterState()
   &SpawnBurst_Instantaneous(Use Spawn Probability=false, UseLoopCountLimit=false)
  ]
  particleSpawnScript [
   &InitializeParticle(
    Color=&FNTimeOfDayColor(
     Night Color=<rgba 0.3116 0.3916 1 1.5>
     Morning Color=<rgba 0.8974 0.4935 0.4987 1.5>
     Day Color=<rgba 1 1 1 1.4>
     Evening Color=<rgba 0.8453 0.5816 0.8974 1.5>
    )
    Lifetime=$Emitter.CurrentLoopDuration
    Mesh Scale=<1.8 1.8 1.8>
    Mesh Scale Mode=::NewEnumerator4
    Position Offset=<0 -85 135>
    UsePositionOffset=true
    Write Material Random=true
   )
   &InitialMeshOrientation(Mesh Orientation Mode=::NewEnumerator2, Orientation Vector=<-1 0 0>, Rotation=<0.11 -0.02 0>, Use Rotation Vector=true)
  ]
  particleUpdateScript [
   &ParticleState()
   &ScaleMeshSize(Scale Factor=&Multiply_VectorByFloat(Vector=<1 1 1>, Float=$User.UseIdleVFX))
   &SolveForcesAndVelocity(Clamp Velocity=false, Limit Acceleration=false)
  ]
  rendererClasses [&NiagaraMeshRendererProperties]
  renderers [{rendererClass=&NiagaraMeshRendererProperties, rendererIndex=0}]
 }
 WindstreaksJet {
  emitterUpdateScript [
   &EmitterState()
   &SpawnByTrigger(
    Spawn Count=10
    Trigger:NiagaraBool=&SetBoolByFloatComparison(A=&VectorLength(Vector=$User.OwnerLocalVelocity), B=20)
    TriggerIntervalSwitch=false
   )
   &SpawnByTrigger(
    Spawn Count=10
    Trigger:NiagaraBool=&SetBoolByFloatComparison(A=$User.FireIntensity, B=0.01)
    TriggerIntervalSwitch=false
   ) {module=SpawnByTrigger001}
  ]
  particleSpawnScript [
   /SetVariables(Particles.SpawnGroup=$Emitter.SpawnGroup) {isSetParameters=true}
   &InitializeParticle(
    Color Maximum=<rgba 0.1009 0.1615 2 0.8609>
    Color Minimum=<rgba 0.541 0.1771 2 0.8609>
    Color Mode=::NewEnumerator2
    Lifetime=&MergeHelper_FloatSwitch(
     0=&UniformRangedFloat(Minimum=0.2, Maximum=0.35)
     1=&UniformRangedFloat(Minimum=0.15, Maximum=0.25)
     2=?
     3=?
     Index=$Particles.SpawnGroup
     Num=::NewEnumerator0
    )
    Sprite Size=&UniformRangedVector2D(Minimum=<20 40>, Maximum=<30 50>)
    Sprite Size Mode=::NewEnumerator4
    UsePositionOffset=false
    Write Material Random=true
   )
   &SkeletalMeshLocation(
    Bone / Socket Index=2
    Bone Sampling Mode=::NewEnumerator4
    MaskSpawnGroup=false
    Sampled Position Offset=<0 0 -5>
    Skeletal Mesh=$System.Skeletal Mesh Reader
   )
   &ShapeLocation(MaskSpawnGroup=false, Override Seed=false, Sphere Radius=40, Sphere Surface Distribution=1, UseOwnerScale=true)
   &AddVelocity(Velocity=&RandomRangeVector(Minimum=<0 200 0>, Maximum=<0 800 0>)) {module=AddVelocity001}
   &AddVelocity(
    Module.Use Falloff Distance=false
    Origin Offset=<0 -100 0>
    Override Seed=false
    Velocity Mode=::NewEnumerator1
    Velocity Speed=&RandomRangeFloat(Minimum=150, Maximum=300)
   )
   &ApplyInitialForces()
   &SubUVAnimation(
    Evaluation Type=::NewEnumerator0
    "Sprite Renderer "=DataInterface() {propertyValues {SpriteRenderer=&self:NiagaraEmitter_8.NiagaraSpriteRendererProperties_0}}
    SubUV Animation Mode:ENiagara_SubUVLookupModeV2=::NewEnumerator2
    UseEndFrame=false
    UseStartFrame=false
   )
  ]
  particleUpdateScript [
   &ParticleState()
   &ScaleSpriteSize(Scale Factor=&Multiply_Vector2D_ByFloat(
     Float=&MakeCustomFloatFromBool(
      Bool=&SetBoolByFloatComparison(A=$User.UseIdleVFX, B=1, Comparison Type=::NewEnumerator6)
      True Float=0
      False Float=1
     )
     Vector2D=&Vector2DFromCurve(
      Vector2Curve=DataInterface() {propertyValues {
        XCurve {keys [
          {arriveTangent=2.5428, interpMode=RCIM_Cubic, leaveTangent=2.5428, tangentMode=RCTM_User, time=0, value=1}
          {arriveTangent=0.1281, interpMode=RCIM_Cubic, leaveTangent=0.1281, tangentMode=RCTM_User, time=1, value=2}
         ]}
        YCurve {keys [
          {arriveTangent=2.8823, interpMode=RCIM_Cubic, leaveTangent=2.8823, tangentMode=RCTM_User, time=0, value=1}
          {arriveTangent=2.0297, interpMode=RCIM_Cubic, leaveTangent=2.0297, tangentMode=RCTM_User, time=1, value=2}
         ]}
       }}
      ScaleCurve=&Multiply_Vector2D_ByFloat(
       Float=&RandomRangeFloat(Minimum=1, Maximum=2)
       Vector2D=&RandomRangeVector2D(Minimum=<0.6 0.75>, Maximum=<0.8 1>)
      )
     )
    ))
   &DynamicMaterialParameters(
    Index 0 Param 1=20
    Index 0 Param 2=1
    Index 0 Param 3=1
    Index 0 Param 4=&FloatFromCurve(FloatCurve=DataInterface() {propertyValues {Curve {keys [{tangentMode=RCTM_User, time=0.1, value=0}, {tangentMode=RCTM_User, time=1, value=0.5}]}}})
    Module.Param0WriteEnabled=true
    Module.Param1WriteEnabled=true
    Module.Param2WriteEnabled=true
    Module.Param3WriteEnabled=true
   ) {module=DynamicMaterialParameters001}
   &Drag(Drag=5, Use Linear Drag=true, Use Rotational Drag=true)
   &SolveForcesAndVelocity(Clamp Velocity=false, Limit Acceleration=false)
   &BlendSpace("Position Blend (World <-> Local)":NiagaraFloat=&FloatFromCurve(
     FloatCurve=DataInterface() {propertyValues {Curve {
        defaultValue=1
        keys [{interpMode=RCIM_Cubic, tangentMode=RCTM_User, time=0, value=1}, {interpMode=RCIM_Cubic, tangentMode=RCTM_User, time=1, value=0}]
       }}}
     Translation Contribution:NiagaraFloat=1
     Rotation Contribution:NiagaraFloat=1
    ))
   &ScaleColor(
    Scale Alpha=&FloatFromCurve(FloatCurve=DataInterface() {propertyValues {Curve {keys [
         {arriveTangent=9.0725, interpMode=RCIM_Cubic, leaveTangent=9.0725, tangentMode=RCTM_User, time=0, value=0}
         {arriveTangent=-2.5249, interpMode=RCIM_Cubic, leaveTangent=-2.5249, tangentMode=RCTM_User, time=0.1408, value=0.9847}
         {arriveTangent=-0.0405, interpMode=RCIM_Cubic, leaveTangent=-0.0405, tangentMode=RCTM_User, time=1, value=0}
        ]}}})
    Scale RGB=&MakeVectorFromLinearColorRGB(Color=&Multiply_LinearColorByFloat(
      Linear Color=<rgba 1 1 1 1>
      Float=&MergeHelper_FloatSwitch(0=1, 1=2, 2=?, 3=?, Index=$Particles.SpawnGroup, Num=::NewEnumerator0)
     ))
    ScaleA=true
    ScaleRGB=true
   )
  ]
  rendererClasses [&NiagaraSpriteRendererProperties]
  renderers [{rendererClass=&NiagaraSpriteRendererProperties}]
 }
}
summary {
 systemName=NS_Emote_SpireTendFlight
 emitters [
  {emitterName=Thruster_Loop, rendererClasses [&NiagaraMeshRendererProperties, &NiagaraSpriteRendererProperties]}
  {emitterName=Thruster_Burst, rendererClasses [&NiagaraMeshRendererProperties]}
  {emitterName=SpeedLines, rendererClasses [&NiagaraSpriteRendererProperties]}
  {emitterName=Smoke, rendererClasses [&NiagaraSpriteRendererProperties]}
  {emitterName=LargeWake, rendererClasses [&NiagaraMeshRendererProperties]}
  {emitterName=WindIdle, rendererClasses [&NiagaraMeshRendererProperties]}
  {emitterName=Light, rendererClasses [&NiagaraLightRendererProperties]}
  {emitterName=WindstreaksJet, rendererClasses [&NiagaraSpriteRendererProperties]}
  {
   emitterName=WindBurstCondensed
   rendererClasses [&NiagaraMeshRendererProperties, &NiagaraSpriteRendererProperties]
  }
  {emitterName=BurstElements, rendererClasses [&NiagaraMeshRendererProperties]}
  {
   emitterName=Thruster_Loop001
   simTarget=GPUComputeSim
   rendererClasses [&NiagaraMeshRendererProperties, &NiagaraSpriteRendererProperties]
  }
 ]
 userVariables [
  User.bIsFrontEnd {type=NiagaraBool}
  User.bIsFrontEndPreview {type=NiagaraBool}
  User.FireIntensity {type=NiagaraFloat, defaultValue=0, description=""}
  User.JetActiveCurve {type=NiagaraFloat, defaultValue=5, description=""}
  User.OwnerLocalVelocity {type=Vector3f, description="", defaultValue {struct=Vector3f, value=<0 0 0>}}
  User.UseIdleVFX {type=NiagaraFloat, defaultValue=0, description=""}
 ]
}
